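# Sound cue for the thrust sound of the Quarrel glider cosmetic.
# A quality-level switch sends its first two levels to a random pick of five thrust waves and its last two to a smaller pool of three.
# Each random pool feeds through a modulator and is excluded from branch culling.
# Glider-specific attenuation, sound class, concurrency and a cosmetic glider loop properties sheet govern playback at reduced volume.
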SoundCue /BRCosmetics/Sounds/Gliders/Quarrel/Glider_Quarrel_M_Thrust_Cue
root=QualityLevel_0
nodes [
 {id=Modulator_0, children [Random_0]}
 {id=Modulator_1, children [Random_2]}
 {id=QualityLevel_0, children [Modulator_0,Modulator_0,Modulator_1,Modulator_1]}
 {
  id=Random_0
  children [WavePlayer_0,WavePlayer_4,WavePlayer_5,WavePlayer_6,WavePlayer_7]
  properties {bShouldExcludeFromBranchCulling=true, weights [1,1,1,1,1]}
 }
 {id=Random_2, children [WavePlayer_8, WavePlayer_9, WavePlayer_10], properties {bShouldExcludeFromBranchCulling=true, weights [1,1,1]}}
 {id=WavePlayer_0, properties {soundWaveAssetPtr=&/Glider_Quarrel_M_Thrust_01}}
 {id=WavePlayer_10, properties {soundWaveAssetPtr=&/Glider_Quarrel_M_Thrust_03}}
 {id=WavePlayer_4, properties {soundWaveAssetPtr=&/Glider_Quarrel_M_Thrust_02}}
 {id=WavePlayer_5, properties {soundWaveAssetPtr=&Glider_Quarrel_M_Thrust_03}}
 {id=WavePlayer_6, properties {soundWaveAssetPtr=&/Glider_Quarrel_M_Thrust_04}}
 {id=WavePlayer_7, properties {soundWaveAssetPtr=&/Glider_Quarrel_M_Thrust_05}}
 {id=WavePlayer_8, properties {soundWaveAssetPtr=&Glider_Quarrel_M_Thrust_01}}
 {id=WavePlayer_9, properties {soundWaveAssetPtr=&Glider_Quarrel_M_Thrust_02}}
]
properties {
 attenuationSettings=&Fort_Player_Glider_Thrust_Athena_Att
 audioPropertiesSheet=&APS_Cosmetic_Glider_Loop
 maxDistance=12500
 soundClassObject=&Fort_Gliders_Land
 volumeMultiplier=0.2
 concurrencySet [&Athena_Glider_Thrust_Conc]
}
